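; Heist Bases environmental light Blueprint that changes color when a lockdown alarm is active.
; Lockdown initiation and expiry set and clear bAlarmActive, whose OnRep handler recolors the point light between muted red for alarm and dim blue for normal.
; The color change runs only where cosmetic events can execute, and expiry ignores its DisableVisuals and DisableSounds parameters.
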(event EventOnLockdownExpired
  (params (DisableVisuals :type Boolean) (DisableSounds :type Boolean))
  (Variables|Default|SetbAlarmActive false) ; replicate lockdown state for client cosmetics
)

(fn OnLockdownInitiated
  (Variables|Default|SetbAlarmActive true)
  (return true))

(fn OnRep_bAlarmActive
  (Networking|CanExecuteCosmeticEvents
    (:True
      (Rendering|Components|Light|SetLightColor (Variables|Default|GetPointLight) (select (Variables|Default|GetbAlarmActive) "(R=0.562500,G=0.152344,B=0.152344,A=1.000000)" "(R=0.094293,G=0.168635,B=0.229167,A=1.000000)") true))))

(fn ConstructionScript)
